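# Unlit translucent material for Niagara mesh particles, drawing a glowing skull in the DarkStance skydiving contrail, tinted green by default.
# Emissive combines a tinted skull texture masked through UV1 with a Fresnel rim, adds particle color, then passes through MF_FX_EmissiveAdjustments.
# Dynamic parameters drive face glow, rim intensity and top fade.
# Opacity combines particle alpha, rerouted masks and depth fade, which Low and Medium quality skip.
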Material /BRCosmetics/Athena/Effects/Particles/Skydiving/NS_Contrail_DarkStance/M_Contrail_DarkStance_Skull
graph {
 properties {
  bUsedWithNiagaraMeshParticles=true
  blendMode=BLEND_Translucent
  pixelDepthOffsetMode=PDOM_Legacy
  shadingModel=MSM_Unlit
  usedShadingModels=Unlit
 }
 nodes [
  n1 = TextureSample(UVs=TextureCoordinate() {coordinateIndex=1}) {texture=&/T_Skull_Mask}
  n2 = Multiply(A=@n1.G, B=ScalarParameter() {defaultValue=1, parameterName=Glow Intensity})
  n3 = Multiply(
   A=TextureSample() {texture=/Game/Effects/Fort_Effects/GameplayPulginsAssets/Skull/SkullDetail/Texture/T_DetailSkull_E}.RGB
   B=VectorParameter() {parameterName=Emissive Tint, defaultValue=<rgba 0.2217 1 0.2 1>}.RGB
  )
  n4 = ScalarParameter() {defaultValue=3, parameterName=Emissive}
  n5 = DynamicParameter() {defaultValue=<rgba 1 1 0.5 1>, paramNames [Fresnel Intensity, Face Intensity, Top Fade Control, Param4]}
  n6 = Multiply(A=Multiply(A=Add(A=@n1.R, B=@n2), B=Multiply(A=@n3, B=@n4) {constB=2}), B=@n5.Face Intensity)
  n7 = ScalarParameter() {defaultValue=4, parameterName=Fresnel Emissive}
  n8 = ScalarParameter() {defaultValue=6, parameterName=Fresnel Exponent}
  n9 = Multiply(A=Multiply(A=Multiply(A=@n3, B=@n7) {constB=2}, B=Fresnel(ExponentIn=@n8) {baseReflectFraction=0}), B=@n5.Fresnel Intensity)
  MaterialFunctionCall(Emissive=Add(A=Add(A=@n6, B=@n9), B=ParticleColor().RGB)) {materialFunction=&MF_FX_EmissiveAdjustments}
  n10 = Multiply(A=ParticleColor().A, B=ScalarParameter() {defaultValue=3, parameterName=Opacity Mult})
  n11 = NamedRerouteUsage()
  n12 = DepthFade(FadeDistance=ScalarParameter() {defaultValue=15, parameterName=FadeDistance})
  n13 = Constant() {r=1}
  Multiply(A=Multiply(A=Multiply(A=@n10, B=@n11), B=NamedRerouteUsage()), B=QualitySwitch(Default=@n12, Low=@n13, High=@n12, Medium=@n13, Epic=@n12))
  n14 = DynamicParameter() {defaultValue=<rgba 1 1 0.5 1>, paramNames [Fresnel Intensity, Face Intensity, Top Fade Control, Param4]}
  n15 = Add(A=@n14.Top Fade Control, B=Constant() {r=0.5})
  n16 = DotProduct(A=Transform(None=Constant3Vector() {constant=<rgba 0 0 1 1>}) {transformSourceType=TRANSFORMSOURCE_Local}, B=VertexNormalWS())
  n17 = OneMinus(None=Saturate(None=ConstantBiasScale(None=@n16)))
  NamedRerouteDeclaration(None=Saturate(None=SmoothStep(Min=@n14.Top Fade Control, Max=@n15, Value=@n17))) {name=Top Opacity Mask, nodeColor=<rgba 1 0 0.9141 1>}
  n18 = Constant() {r=4.4707}
  n19 = Add(A=@n18, B=Constant() {r=1})
  n20 = MaterialFunctionCall(Float3=MaterialFunctionCall() {materialFunction=&LocalPosition}.Instance Local Position) {materialFunction=&BreakOutFloat3Components}
  NamedRerouteDeclaration(None=Saturate(None=SmoothStep(Min=@n18, Max=@n19, Value=@n20.G))) {name=SolidFront, nodeColor=<rgba 0.6641 0 1 1>}
  n21 = Constant() {r=0.6}
  n22 = CameraVectorWS()
  n23 = VertexNormalWS()
  n24 = SmoothStep(Min=@n21, Max=Add(A=@n21, B=Constant() {r=0.4}), Value=ConstantBiasScale(None=DotProduct(A=@n22, B=@n23)))
  n25 = ScalarParameter() {defaultValue=0.8, parameterName=Fresnel Base}
  NamedRerouteDeclaration(None=Saturate(None=LinearInterpolate(B=@n24, Alpha=@n25) {constA=1})) {name=SoftFresnel, nodeColor=<rgba 0.1719 1 0 1>}
 ]
}
