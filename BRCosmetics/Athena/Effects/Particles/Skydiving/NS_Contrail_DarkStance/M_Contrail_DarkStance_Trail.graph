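# Unlit, two-sided translucent material for the trail of the DarkStance skydiving contrail, built for Niagara mesh particles.
# Panning smoke, trail and caustic textures, distorted and eroded by noise, form a mask alongside particle color and near-camera and view-angle fades.
# World position offset nodes combine spline thickening, side/back offsets and a lag bend weighted along the trail's UVs.
# Dynamic parameters expose U scale, U offset, forward-tail masking and side offset.
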Material /BRCosmetics/Athena/Effects/Particles/Skydiving/NS_Contrail_DarkStance/M_Contrail_DarkStance_Trail
graph {
 properties {
  bUsedWithNiagaraMeshParticles=true
  blendMode=BLEND_Translucent
  pixelDepthOffsetMode=PDOM_Legacy
  shadingModel=MSM_Unlit
  twoSided=true
  usedShadingModels=Unlit
 }
 nodes [
  Add(A=NamedRerouteUsage(), B=NamedRerouteUsage()) {constB=0}
  Add(A=WorldPosition().XYZ, B=Transform(None=NamedRerouteUsage()) {transformSourceType=TRANSFORMSOURCE_Local})
  Constant() {r=-0.5}
  Constant() {r=1.5}
  Constant() {r=-1}
  Constant() {r=0.5}
  MaterialFunctionCall(Emissive=ParticleColor().RGB) {materialFunction=&MF_FX_EmissiveAdjustments}
  n1 = NamedRerouteUsage()
  n2 = Power(
   Base=MaterialFunctionCall(Distance=ScalarParameter() {defaultValue=100}) {materialFunction=&Near_Camera_Fade}.Result
   Exp=ScalarParameter() {defaultValue=3, parameterName=NearCamFade}
  )
  Multiply(
   A=Multiply(A=Multiply(A=@n1, B=@n2), B=ParticleColor().A)
   B=MaterialFunctionCall("FadeSensitivity (optional)"=ScalarParameter() {parameterName=FadeSensitivity}) {materialFunction=&FadeBasedOnViewAngle}.Result
  )
  n3 = WorldPosition()
  n4 = NamedRerouteUsage()
  n5 = Normalize(VectorInput=Subtract(A=CameraPositionWS(), B=Add(A=@n3.XYZ, B=@n4)))
  n6 = Normalize(VectorInput=Transform(None=Constant3Vector() {constant=<rgba 0 1 0 0>}))
  n7 = ComponentMask(None=TextureCoordinate()) {r=true}
  n8 = Multiply(B=ScalarParameter() {defaultValue=80, parameterName=Width}) {constA=1}
  Multiply(A=Normalize(VectorInput=CrossProduct(A=@n5, B=@n6)), B=Multiply(A=Subtract(A=@n7) {constB=0.5}, B=@n8))
  Multiply(A=ComponentMask(None=TextureCoordinate() {coordinateIndex=1}) {r=true}, B=@n8) {desc="Width information\r\nin UV1.x"}
  NamedRerouteDeclaration(None=MaterialFunctionCall(WidthBase=ScalarParameter() {defaultValue=40, parameterName=Base}, WidthTip=ScalarParameter() {defaultValue=100, parameterName=Tip}) {materialFunction=&SplineThicken}.WorldPositionOffset) {name=SplineThicken, nodeColor=<rgba 0 1 1 1>}
  n9 = TextureSample(UVs=Panner(Coordinate=NamedRerouteUsage()) {speedX=1}) {samplerType=SAMPLERTYPE_Grayscale, texture=&T_Smoke_Trail_Soft}
  n10 = NamedRerouteUsage()
  n11 = Panner(Coordinate=NamedRerouteUsage()) {speedX=0.1, speedY=0.2}
  n12 = Multiply(A=ComponentMask(None=TextureSample(UVs=@n11) {samplerType=SAMPLERTYPE_Masks, texture=/Game/Weapons/WeaponSkins/Wraps/Textures/Mash/Mash_Noise}.RGB) {g=true, r=true}) {constB=0.12}
  n13 = TextureSample(UVs=Panner(Coordinate=Add(A=@n10, B=@n12)) {speedX=1}) {texture=/Game/Gadgets/Assets/VinderTech_GliderChute/Glider_DarkStance/FX/T_Glider_DarkStance_Trail}
  n14 = Power(Base=Add(A=Multiply(A=@n9.R), B=Multiply(A=@n13.G) {constB=2}) {constB=0})
  n15 = DynamicParameter() {defaultValue=<rgba 200 50 1 0>, paramNames [WidthBase, WidthTip, U_Scale, U offset]}
  n16 = Constant() {r=1}
  n17 = Add(A=Multiply(A=AppendVector(A=@n15.U_Scale, B=@n16), B=NamedRerouteUsage()), B=@n12)
  n18 = TextureSample(UVs=Panner(Coordinate=NamedRerouteUsage()) {speedX=0.5}) {
   samplerType=SAMPLERTYPE_LinearColor
   texture=/Engine/Functions/Engine_MaterialFunctions02/ExampleContent/Textures/OffsetNoiseDistanceFields_ForNormals
  }
  n19 = Multiply(
   A=TextureSample(UVs=Panner(Coordinate=@n17) {speedX=2}) {texture=&T_Glider_DarkStance_Trail}.R
   B=LinearInterpolate(Alpha=@n18.R) {constA=0.4}
  )
  n20 = Saturate(None=Add(A=Add(A=@n14, B=@n19)) {constB=0})
  n21 = Power(Base=ComponentMask(None=NamedRerouteUsage()) {r=true}) {constExponent=1.5}
  n22 = Multiply(A=@n20, B=Power(Base=LinearInterpolate(Alpha=@n18.R) {constA=0.5}, Exp=LinearInterpolate(Alpha=@n21)) {constExponent=1})
  n23 = LinearInterpolate(Alpha=ComponentMask(None=NamedRerouteUsage()) {r=true}) {constA=0.8}
  n24 = Multiply(
   A=AppendVector(A=DynamicParameter() {defaultValue=<rgba 200 50 1 0>, paramNames [WidthBase, WidthTip, U_Scale, U offset]}.U_Scale, B=Constant() {r=1})
   B=NamedRerouteUsage()
  )
  n25 = TextureSample(UVs=Add(A=Panner(Coordinate=@n24) {speedX=1}, B=Multiply(A=@n12))) {samplerType=SAMPLERTYPE_LinearGrayscale, texture=/Game/Global/Textures/Default/Caustics/T_Nosie_Caustics_Grayscale}
  n26 = ComponentMask(None=NamedRerouteUsage()) {r=true}
  n27 = Multiply(A=Multiply(A=Power(Base=@n25.R) {constExponent=1}) {constB=1.5}, B=Power(Base=OneMinus(None=@n26)) {constExponent=1})
  n28 = Saturate(None=Add(A=Multiply(A=@n22, B=@n23), B=Multiply(A=@n27, B=@n20)))
  n29 = DynamicParameter() {parameterIndex=1, defaultValue=<rgba 1 1 0 0>, paramNames [Use Forward Tail mask, Appear Mask, none, none]}
  n30 = NamedRerouteUsage()
  n31 = AppendVector(A=DynamicParameter() {defaultValue=<rgba 200 50 1 0>, paramNames [WidthBase, WidthTip, U_Scale, U offset]}.U offset, B=Constant())
  n32 = TextureSample(UVs=Panner(Coordinate=Add(A=@n30, B=@n31)) {speedX=0.5}) {samplerType=SAMPLERTYPE_Masks, texture=&Mash_Noise}
  n33 = DynamicParameter() {parameterIndex=1, defaultValue=<rgba 1 1 0 0>, paramNames [Use Forward Tail mask, Appear Mask, none, none]}
  n34 = ComponentMask(None=NamedRerouteUsage()) {r=true}
  n35 = Saturate(None=LinearInterpolate(
    A=Constant() {r=-1}
    B=LinearInterpolate(Alpha=@n33.Use Forward Tail mask) {constA=0.5, constB=1.5}
    Alpha=Power(Base=@n34) {constExponent=1.3}
   ) {constA=0.1, constB=2})
  n36 = Saturate(None=LinearInterpolate(A=LinearInterpolate(Alpha=@n29.Use Forward Tail mask) {constB=-0.2}, B=Constant() {r=1.2}, Alpha=Power(Base=@n32.R, Exp=@n35)) {constA=-2, constB=2})
  n37 = NamedRerouteUsage()
  n38 = Multiply(A=Subtract(A=ComponentMask(None=@n37) {r=true}) {constB=0.1}) {constB=3}
  NamedRerouteDeclaration(None=Multiply(A=Multiply(A=@n28, B=@n36), B=Saturate(None=@n38))) {name=test, nodeColor=<rgba 0 0.1328 1 1>}
  NamedRerouteDeclaration(None=MaterialFunctionCall(XY=TextureCoordinate()) {materialFunction=&Swizzle}.YX) {name=UV, nodeColor=<rgba 0 0.1797 1 1>}
  n39 = NamedRerouteUsage()
  n40 = NamedRerouteUsage()
  n41 = MaterialFunctionCall(Float2=NamedRerouteUsage()) {materialFunction=&BreakOutFloat2Components}
  n42 = Multiply(
   A=MaterialFunctionCall(
    X=DynamicParameter() {parameterIndex=2, defaultValue=<rgba 0 50 1 0>, paramNames [Side WPO Offset, "", "", ""]}.Side WPO Offset
    Y=ScalarParameter() {defaultValue=-50, parameterName=Back Offset}
    Z=Constant()
   ) {materialFunction=&MakeFloat3}.Result
   B=Power(Base=OneMinus(None=@n41.R)) {constExponent=1.6}
  )
  NamedRerouteDeclaration(None=Add(A=Add(A=@n39, B=@n40) {constB=0}, B=Transform(None=@n42) {transformSourceType=TRANSFORMSOURCE_Local}) {constB=0}) {name=WPO, nodeColor=<rgba 0.1953 0 1 1>}
  n43 = DynamicParameter() {parameterIndex=1, defaultValue=<rgba 1 1 0 0>, paramNames [Use Forward Tail mask, Appear Mask, none, none]}
  n44 = Constant()
  n45 = Clamp(None=Divide(A=MaterialFunctionCall(R=@n43.B, G=@n43.B, B=@n44) {materialFunction=&AppendMany}.RGB) {constB=-960}) {minDefault=-1}
  n46 = ScalarParameter() {defaultValue=50, parameterName=LagStrength}
  n47 = ComponentMask(None=NamedRerouteUsage()) {r=true}
  n48 = Saturate(None=Power(Base=OneMinus(None=@n47), Exp=ScalarParameter() {defaultValue=1.66, parameterName=WPO Bend Power}) {constExponent=3})
  n49 = ScalarParameter() {defaultValue=1, parameterName=LagPowerA}
  NamedRerouteDeclaration(None=Multiply(A=Multiply(A=@n45, B=@n46), B=Multiply(A=@n48, B=@n49))) {name=VertexLag, nodeColor=<rgba 1 0.3281 0 1>}
  NamedRerouteUsage()
  ScalarParameter() {defaultValue=80, parameterName=SideOffset}
  StaticBool() {value=true}
  TextureCoordinate()
  TextureCoordinate()
  TextureCoordinate()
  TextureCoordinate()
  TextureCoordinate() {vTiling=0.6}
  TextureCoordinate()
  TextureCoordinate() {uTiling=0.8, vTiling=0.5}
  TextureCoordinate()
  TextureCoordinate() {uTiling=2}
  TextureCoordinate()
  TextureCoordinate()
  TextureCoordinate() {uTiling=1.1, vTiling=1.7}
  TextureCoordinate()
  Transform(None=MaterialFunctionCall() {materialFunction=&LocalPosition}.Local Position) {transformSourceType=TRANSFORMSOURCE_Local}
 ]
}
