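# Skydiving contrail cosmetic for DarkStance: six skull meshes hover around the player, trailing smoke, sparks and energy.
# SKULLS follows the skeletal mesh with left/right ring offsets, waveform jitter and smoothed lag.
# Dark smoke, green puffs, sparks and trail meshes read SKULLS particles through attribute readers, with DoOnce triggering puffs and spark bursts.
# A smoothed system value drives motion and intensity, with front-end overrides and trail shaping via dynamic material parameters.
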NiagaraSystem /BRCosmetics/Athena/Effects/Particles/Skydiving/NS_Contrail_DarkStance/NS_DarkStance_Contrail
dependencies {
 usedDataInterfaces [
  &NiagaraDataInterfaceParticleRead
  &NiagaraDataInterfaceCurve
  &NiagaraDataInterfaceArrayFloat3
  &NiagaraDataInterfaceColorCurve
  &NiagaraDataInterfaceSpriteRendererInfo
  &NiagaraDataInterfaceArrayFloat
  &NiagaraDataInterfacePlatformSet
  &NiagaraDataInterfaceSkeletalMesh
 ]
 usedDynamicInputs [&SetBoolByFloatComparison, &Divide_Float, &Abs_Float]
 usedModules [
  &self:SystemSpawnScript.SystemScriptSource.SystemScriptGraph.NiagaraNodeAssignment_0.SetVariables_#1
  &SystemState
  &self:SystemSpawnScript.SystemScriptSource.SystemScriptGraph.NiagaraNodeAssignment_4.SetVariables_#2
  &self:SystemSpawnScript.SystemScriptSource.SystemScriptGraph.NiagaraNodeAssignment_1.SetVariables_#3
  &FN_Integrator
  &Timeline
  &self:SystemSpawnScript.SystemScriptSource.SystemScriptGraph.NiagaraNodeAssignment_2.SetVariables_#4
  &self:SystemSpawnScript.SystemScriptSource.SystemScriptGraph.NiagaraNodeAssignment_3.SetVariables_#5
  &DoOnce
 ]
}
emitters {
 Blend_Smoke {
  emitterUpdateScript [
   &EmitterState()
   &SpawnRate(
    Spawn Probability=&MakeFloatFromInt(Int=&MakeCustomIntFromBool(Bool=$User.bIsFrontEndPreview, True INT=1, False INT=1))
    SpawnRate=&Multiply_Float(A=50, B=&Lerp_Float(A=1, B=1.33, Alpha=$System.InterpolateOverTime.InterpolateOverTime_Float.Moving Average))
    Use Spawn Probability=true
   )
  ]
  particleSpawnScript [
   &FN_SetSpawnGroup(Number of Spawn Groups=6)
   &InitializeParticle(
    Color=<rgba 0.009 0.0092 0.01 5>
    Lifetime=&Multiply_Float(A=0.5, B=&Lerp_Float(A=1, B=0.6, Alpha=$System.InterpolateOverTime.InterpolateOverTime_Float.Moving Average))
    Sprite Rotation Mode=::NewEnumerator2
    Sprite Size Mode=::NewEnumerator1
    Uniform Sprite Size Max=50
    Uniform Sprite Size Min=40
    UsePositionOffset=false
    Write Material Random=true
   )
   &self:SampleSourceSpawnGroupFromOtherParticle(PAR=DataInterface() {propertyValues {EmitterBinding {bindingMode=Other, emitterName=SKULLS}}}) {module=SampleSourceSpawnGroupFromOtherParticle}
  ]
  particleUpdateScript [
   &self:SamplePositionFromOtherParticle(PAR=DataInterface() {propertyValues {EmitterBinding {bindingMode=Other, emitterName=SKULLS}}}) {module=SamplePositionFromOtherParticle}
   /SetVariables(Particles.Position=&AddVectorToPosition(
     Position=$Particles.Position
     Vector=&TransformVector(
      Vector=&MakeVector(
       X=0
       Y=0
       Z=&Lerp_Float(
        A=3
        B=&Lerp_Float(A=25, B=40, Alpha=$System.InterpolateOverTime.InterpolateOverTime_Float.Moving Average)
        Alpha=$Particles.NormalizedAge
       )
      )
      Destination Space=::Simulation
     )
    )) {isSetParameters=true, module=SetVariables_#6}
   &ParticleState()
   &ScaleSpriteSize(
    Uniform Curve Scale=&Multiply_Float(
     A=&Lerp_Float(A=1, B=1.2, Alpha=$System.InterpolateOverTime.InterpolateOverTime_Float.Moving Average)
     B=&Multiply_Float(A=$Particles.SkullScale, B=0.9)
    )
    Uniform Curve Sprite Scale=DataInterface() {propertyValues {Curve {keys [{arriveTangent=1.6237, time=0, value=0.6}, {arriveTangent=0.637, time=1, value=1.1}]}}}
   )
   &SpriteRotationRate(Rotation Rate=&RandomRangeFloat(Minimum=-150, Maximum=150))
   &SolveForcesAndVelocity(Clamp Velocity=false, Limit Acceleration=false)
   &ScaleColor(
    Scale Alpha=&FloatFromCurve(
     FloatCurve=DataInterface() {propertyValues {Curve {keys [
         {arriveTangent=0.9905, interpMode=RCIM_Cubic, leaveTangent=0.9905, tangentMode=RCTM_User, time=0, value=0}
         {arriveTangent=-0.5, interpMode=RCIM_Cubic, leaveTangent=-3.148, tangentMode=RCTM_Break, time=0.25, value=1}
         <at 1 0 RCIM_Cubic>
        ]}}}
     Scale Curve=&Multiply_Float(
      A=&Lerp_Float(
       A=1
       B=&FN_FrontEndFloatModifier(Gameplay=1, Front End=1)
       Alpha=$System.InterpolateOverTime.InterpolateOverTime_Float.Moving Average
      )
      B=&FloatFromCurve(FloatCurve=DataInterface() {propertyValues {Curve {keys [<at 0 0 RCIM_Cubic>,<at 0.4 1 RCIM_Cubic>]}}}, CurveIndex=$System.Age)
     )
    )
    ScaleA=true
    ScaleRGB=true
   )
   &CameraOffset(Camera Offset Amount=-5)
  ]
  rendererClasses [&NiagaraSpriteRendererProperties]
  renderers [{rendererClass=&NiagaraSpriteRendererProperties}]
 }
 SKULLS {
  emitterUpdateScript [
   &EmitterState(
    Inactive Response=::NewEnumerator1
    Life Cycle Mode=::NewEnumerator1
    Loop Behavior=::NewEnumerator1
    Loop Duration Mode=::NewEnumerator1
    UseLoopDelay=false
   )
   &SpawnBurst_Instantaneous(
    Spawn Count=&MakeIntFromFloat(Float=&FN_FrontEndFloatModifier(Gameplay=6, Front End=6))
    Use Spawn Probability=false
    UseLoopCountLimit=true
   )
   &InterpolateOverTime(Rate Of Change=50, Target Vector Value:Vector3f=$Engine.Owner.Velocity, Variable Type=::NewEnumerator2)
  ]
  particleSpawnScript [
   &FN_SetSpawnGroup(Number of Spawn Groups=6) {module=FN_SetSpawnGroup001}
   /SetVariables(Particles.SpawnGroup2:NiagaraInt32=$Particles.SpawnGroup) {isSetParameters=true}
   &FN_SetSpawnGroup(Number of Spawn Groups=2)
   &InitializeParticle(
    Color=<rgba 0.01 0.01 0.01 1.5>
    Lifetime=2
    Mesh Scale=&VectorFromFloat(Value=&FN_FrontEndFloatModifier(
      Gameplay=&SelectFloatFromArray(Direct Array Index=$Particles.SpawnGroup2, Float Selection Array=DataInterface() {propertyValues {FloatData [1.6,1.6,1,1,1,1]}})
      Front End=&SelectFloatFromArray(Direct Array Index=$Particles.SpawnGroup2, Float Selection Array=DataInterface() {propertyValues {FloatData [1.6,1.6,1.1,1.1,1.1,1.1]}})
     ))
    Mesh Scale Mode=::NewEnumerator4
    Sprite Size Mode=::NewEnumerator3
    Uniform Sprite Size=50
    UsePositionOffset=true
    Write Material Random=true
   )
  ]
  particleUpdateScript [
   &ParticleState(Kill Particles When Lifetime Has Elapsed=false, Let Infinitely Lived Particles Die When Emitter Deactivates=true)
   &SkeletalMeshLocation(
    Bone / Socket Index=2
    Bone Sampling Mode=::NewEnumerator4
    MaskSpawnGroup=false
    Orientation Sampling=::NewEnumerator1
    Sampled Orientation Coordinate Space=::World
    Skeletal Mesh=$System.Skeletal Mesh
    Use Sampled Mesh Attribute as Ribbon ID=false
   )
   /SetVariables(Particles.AmbientMovement:Vector3f=&Multiply_VectorByFloat(
     Vector=&MakeVector(
      X=&Waveform(
       Global Amplitude Scale=0.5
       "Waveform [1]"=::NewEnumerator12
       "[1] Input Position"=&RandomRangeFloat(Minimum=0, Maximum=1000)
       "[1] Frequency"=1
       "[1] Amplitude Min/Max"=<-50 50>
      )
      Y=&Waveform(
       "Waveform [1]"=::NewEnumerator12
       "[1] Input Position"=&RandomRangeFloat(Minimum=1200, Maximum=1800)
       "[1] Frequency"=1
       "[1] Amplitude Min/Max"=<-50 50>
      )
      Z=&Waveform(
       Global Amplitude Scale=1.5
       "Waveform [1]"=::NewEnumerator12
       "[1] Input Position"=&RandomRangeFloat(Minimum=2200, Maximum=2800)
       "[1] Frequency"=1
       "[1] Amplitude Min/Max"=<-50 50>
      )
     )
     Float=0.3333
    )) {isSetParameters=true}
   /SetVariables(Particles.HF jiggle:Vector3f=&Multiply_VectorByFloat(
     Vector=&MakeVector(
      X=&Waveform(
       Global Input Position=&RandomRangeFloat(Minimum=0, Maximum=1)
       "Waveform [1]"=::NewEnumerator12
       "[1] Input Position"=&RandomRangeFloat(Minimum=800, Maximum=1600)
       "[1] Frequency"=8
       "[1] Amplitude Min/Max"=<-10 10>
       "[1] Scale"=0.333
      )
      Y=&Waveform(
       Global Input Position=&RandomRangeFloat(Minimum=0, Maximum=1)
       "Waveform [1]"=::NewEnumerator12
       "[1] Input Position"=&RandomRangeFloat(Minimum=1400, Maximum=2000)
       "[1] Frequency"=8
       "[1] Amplitude Min/Max"=<-10 10>
       "[1] Scale"=0.333
      )
      Z=&Waveform(
       Global Input Position=&RandomRangeFloat(Minimum=0, Maximum=1)
       "Waveform [1]"=::NewEnumerator12
       "[1] Input Position"=&RandomRangeFloat(Minimum=0, Maximum=1000)
       "[1] Frequency"=8
       "[1] Amplitude Min/Max"=<-10 10>
       "[1] Scale"=0.333
      )
     )
     Float=&Lerp_Float(A=0, B=0.8, Alpha=$System.InterpolateOverTime.InterpolateOverTime_Float.Moving Average)
    )) {isSetParameters=true}
   &self:ScratchModule_02(InterpolatedQuat:Quat4f=$Particles.SkeletalMeshLocation.SampledRotation, WorldToLocalMatrix:NiagaraMatrix=$Engine.Owner.SystemWorldToLocal) {module=ScratchModule_02}
   &InterpolateOverTime(Rate Of Change=6, Target Quat Value:Quat4f=$Particles.SkeletalMeshLocation.SampledRotation, Variable Type=::NewEnumerator5) {module=InterpolateOverTime001}
   &self:ScratchModule_01(
    InterpolatedQuat:Quat4f=$Particles.InterpolateOverTime001.InterpolateOverTime_Quaternion.Moving Average
    WorldToLocalMatrix:NiagaraMatrix=$Engine.Owner.SystemWorldToLocal
   ) {module=ScratchModule_01}
   /SetVariables(Particles.RotAmount:NiagaraFloat=&self:ScratchDynamicInput_02(
     InitialMeshVector:Vector3f=&ReturnMeshOrientationAxis(Mesh Orientation Axis=::ZAxis, Mesh Orientation Quat=$Particles.InitialMeshOrientation)
     MeshVector:Vector3f=&ReturnMeshOrientationAxis(Mesh Orientation Axis=::ZAxis, Mesh Orientation Quat=$Particles.MeshOrientation)
    )) {isSetParameters=true}
   &ShapeLocation(
    Additional Quaternion Rotation=true
    Additional Yaw / Pitch / Roll=&MakeVector(X=&MaskFloatBySpawnGroup(Masked Float=0.25, Passthrough Float=0, Spawn Group Mask=1), Y=0, Z=0)
    Apply To Particle Position=::NewEnumerator1
    MaskSpawnGroup=false
    Offset=&Lerp_Vector(
     A=&MaskVectorBySpawnGroup(Masked Vector=<70 60 0>, Passthrough Vector=<-70 60 0>)
     B=&FN_FrontEndVector3Modifier(
      Gameplay=&MaskVectorBySpawnGroup(Masked Vector=<80 50 0>, Passthrough Vector=<-80 50 0>)
      Front End=&MaskVectorBySpawnGroup(Masked Vector=<80 50 -30>, Passthrough Vector=<-80 50 -30>)
     )
     Alpha=&ClampFloat(Float=&Add_Float(
       A=$System.InterpolateOverTime.InterpolateOverTime_Float.Moving Average
       B=&Add_Float(
        A=&RemapRange(Input Value To Remap=$Particles.RotAmount, Input Min=15, Input Max=90, Output Min=0, Output Max=1)
        B=$System.Timeline004.NormalizedTime
       )
      ))
    )
    Offset Mode=::NewEnumerator0
    Override Seed=false
    Ring / Disc Distribution Mode:ENiagaraSphereDistributionMode=::NewEnumerator1
    Ring Radius=&FN_FrontEndFloatModifier(
     Gameplay=&Lerp_Float(A=25, B=45, Alpha=$System.InterpolateOverTime.InterpolateOverTime_Float.Moving Average)
     Front End=&Lerp_Float(A=20, B=30, Alpha=$System.InterpolateOverTime.InterpolateOverTime_Float.Moving Average)
    )
    Rotation Mode=::NewEnumerator3
    Rotation Quaternion=&MakeQuatFromAxisAngle(Axis=<1 1 1>, AngleInDegrees=0.2)
    Shape Primitive=::NewEnumerator4
    U Position=&ReturnNormalizedExecIndex(Normalized Index Includes One=false)
    UseOwnerScale=true
   ) {module=ShapeLocation002}
   /SetVariables(Particles.ShapeVectorToWorld:Vector3f=&TransformVector(Vector=$Particles.ShapeLocation002.ShapeVector, Destination Space=::World)) {isSetParameters=true}
   &InterpolateOverTime(
    Rate Of Change:NiagaraFloat=&SelectFloatFromArray(Direct Array Index=$Particles.SpawnGroup2, Float Selection Array=DataInterface() {propertyValues {FloatData [6,10,3,6,10,4]}})
    Target Vector Value:Vector3f=$Particles.ShapeVectorToWorld
    Variable Type=::NewEnumerator2
   ) {module=InterpolateOverTime002}
   /SetVariables(Particles.MeshOrientation=&OrientMeshToVector(Look At Direction=$Particles.ShapeLocation002.ShapeTangent, Reference Axis Vector=<1 0 0>)) {isSetParameters=true}
   /SetVariables(Particles.MeshOrientation=&MultiplyQuaternion(
     Quaternion A=$Particles.MeshOrientation
     Quaternion B=&MakeQuatFromAxisAngle(
      Axis=<0 1 0>
      AngleInDegrees=&Lerp_Float(A=30, B=90, Alpha=$System.InterpolateOverTime.InterpolateOverTime_Float.Moving Average)
     )
    )) {isSetParameters=true}
   /SetVariables(Particles.CatchupVector:Vector3f=&Multiply_VectorByFloat(
     Vector=&Multiply_VectorByFloat(
      Vector=&NormalizeVector(Vector To Normalize=$System.SkydiveVelocity, Safe Normalize=true, Safe Fallback Threshold=0, Safe Fallback Vector=<0 0 0.001>)
      Float=&FN_FrontEndFloatModifier(Gameplay=0, Front End=0)
     )
     Float=&Lerp_Float(
      A=&FloatFromCurve(
       FloatCurve=DataInterface() {propertyValues {Curve {keys [
           {interpMode=RCIM_Cubic, tangentMode=RCTM_User, time=0, value=0}
           {interpMode=RCIM_Cubic, tangentMode=RCTM_User, time=0.35, value=-0.35}
           {tangentMode=RCTM_User, time=1, value=0}
          ]}}}
       CurveIndex=$System.Timeline001.NormalizedTime
      )
      B=&FloatFromCurve(
       FloatCurve=DataInterface() {propertyValues {Curve {keys [
           {arriveTangent=3.3827, arriveTangentWeight=1.1106, interpMode=RCIM_Cubic, leaveTangentWeight=1.1106, time=0, value=0}
           {arriveTangentWeight=0.5241, interpMode=RCIM_Cubic, leaveTangentWeight=0.1667, time=0.5, value=0.65}
           <at 1 0 RCIM_Cubic>
          ]}}}
       CurveIndex=$System.Timeline001.NormalizedTime
      )
      Alpha=$Particles.InterpolateOverTime004.InterpolateOverTime_Float.Moving Average
     )
    )) {isSetParameters=true}
   /SetVariables(Particles.InitialOffset:Vector3f=&Lerp_Vector(
     A=&FN_FrontEndVector3Modifier(Gameplay=<0 0 500>, Front End=<0 0 300>)
     B=&Lerp_Vector(A=<0 0 0>, B=<0 0 0>, Alpha=$System.InterpolateOverTime.InterpolateOverTime_Float.Moving Average)
     Alpha=&FloatFromCurve(
      FloatCurve=DataInterface() {propertyValues {Curve {keys [
          {arriveTangent=1.5577, interpMode=RCIM_Cubic, leaveTangent=1.5577, tangentMode=RCTM_User, time=0, value=0}
          {arriveTangent=-0.0191, interpMode=RCIM_Cubic, leaveTangent=-0.0191, tangentMode=RCTM_User, time=1, value=1}
         ]}}}
      CurveIndex=&Multiply_Float(A=$Engine.System.Age, B=&RandomRangeFloat(Minimum=0.7, Maximum=1))
     )
    )) {isSetParameters=true}
   /SetVariables(Particles.MainOffset:Vector3f=&Add_Vector(
     A=<0 0 0>
     B=&Add_Vector(
      A=$Particles.InitialOffset
      B=&Add_Vector(
       A=$Particles.AmbientMovement
       B=&Add_Vector(A=$Particles.CatchupVector, B=&Add_Vector(A=<0 0 0>, B=$Particles.HF jiggle))
      )
     )
    )) {isSetParameters=true}
   /SetVariables(Particles.Position=&AddVectorToPosition(
     Position=$Particles.SkeletalMeshLocation.SampledPosition
     Vector=&Add_Vector(
      A=&TransformVector(Vector=$Particles.InterpolateOverTime002.InterpolateOverTime_V3_.Moving Average, Source Space=::World, Destination Space=::Local)
      B=$Particles.MainOffset
     )
    )) {isSetParameters=true}
   /SetVariables(Particles.Wobble:Quat4f=&MultiplyQuaternion(
     Quaternion A=&FindQuatBetweenTwoVectors(
      From Vector=&MakeVector(
       X=&Sine(
        Normalized Angle=&Add_Float(A=$System.FN_Integrator.Value, B=&RandomRangeFloat(Minimum=0, Maximum=1))
        Period=&MakeCustomFloatFromBool(Bool=&RandomBool(), True Float=1, False Float=-1)
        Scale=1
        Bias=0
       )
       Y=&Cosine(
        Normalized Angle=&Add_Float(A=$System.FN_Integrator.Value, B=&RandomRangeFloat(Minimum=0, Maximum=1))
        Period=&MakeCustomFloatFromBool(Bool=&RandomBool(), True Float=1, False Float=-1)
        Scale=1
        Bias=0
       )
       Z=&RandomRangeFloat(Minimum=6, Maximum=8)
      )
      To Vector=<0 0 1>
     )
     Quaternion B=$Particles.Initial.MeshOrientation
    )) {isSetParameters=true}
   /SetVariables(Particles.MeshOrientation=&MultiplyQuaternion(Quaternion A=$Particles.MeshOrientation, Quaternion B=$Particles.Wobble)) {isSetParameters=true}
   &DynamicMaterialParameters(
    Index 0 Param 1=&ClampFloat(Float=&Add_Float(
      A=&Lerp_Float(
       A=0.01
       B=&FN_FrontEndFloatModifier(Gameplay=0, Front End=0)
       Alpha=$System.InterpolateOverTime.InterpolateOverTime_Float.Moving Average
      )
      B=0
     ))
    Index 0 Param 2=&Lerp_Float(A=0.2, B=3, Alpha=$System.InterpolateOverTime.InterpolateOverTime_Float.Moving Average)
    Index 0 Param 3=&Lerp_Float(
     A=&FN_PickFloatDependOnPlatform(TargetPlatformActive=0, TargetPlatformNoActive=-0.2)
     B=-0.2
     Alpha=$System.InterpolateOverTime.InterpolateOverTime_Float.Moving Average
    )
    Module.Index0Param0Write=true
    Module.Index0Param1Write=true
    Module.Index0Param2Write=true
    Module.Index0Param3Write=false
   )
   &ScaleColor(
    Scale Alpha=&Multiply_Float(
     A=&FN_FrontEndFloatModifier(Gameplay=0.7, Front End=0.8)
     B=&FloatFromCurve(FloatCurve=DataInterface() {propertyValues {Curve {keys [<at 0 0 RCIM_Cubic>,<at 0.4 1 RCIM_Cubic>]}}}, CurveIndex=$System.Age)
    )
    ScaleA=true
    ScaleRGB=false
   )
   /SetVariables(Particles.Glow Color=&MakeLinearColorFromVectorAndFloat(
     "Vector (RGB)"=&VectorFromFloat(Value=&Lerp_Float(A=0.15, B=0.8, Alpha=$System.InterpolateOverTime.InterpolateOverTime_Float.Moving Average))
     "Float (Alpha)"=1
    )) {isSetParameters=true}
   &SolveForcesAndVelocity(Clamp Velocity=false, Limit Acceleration=false)
   &ScaleMeshSize(Scale Factor=&VectorFromFloat(Value=&FloatFromCurve(
      FloatCurve=DataInterface() {propertyValues {Curve {keys [
          {arriveTangent=1.5577, interpMode=RCIM_Cubic, leaveTangent=1.5577, tangentMode=RCTM_User, time=0, value=0}
          {arriveTangent=-0.0191, interpMode=RCIM_Cubic, leaveTangent=-0.0191, tangentMode=RCTM_User, time=1, value=1}
         ]}}}
      CurveIndex=&Multiply_Float(A=$Engine.System.Age, B=&RandomRangeFloat(Minimum=0.7, Maximum=1))
     )))
  ]
  rendererClasses [&NiagaraMeshRendererProperties]
  renderers [{rendererClass=&NiagaraMeshRendererProperties, rendererIndex=0}]
 }
 Smoke {
  emitterUpdateScript [
   &EmitterState()
   &SpawnRate(
    Spawn Probability=&MakeFloatFromInt(Int=&MakeCustomIntFromBool(Bool=$User.bIsFrontEndPreview, True INT=1, False INT=1))
    SpawnRate=&Multiply_Float(A=50, B=&Lerp_Float(A=1, B=1.33, Alpha=$System.InterpolateOverTime.InterpolateOverTime_Float.Moving Average))
    Use Spawn Probability=true
   )
  ]
  particleSpawnScript [
   &FN_SetSpawnGroup(Number of Spawn Groups=6)
   &InitializeParticle(
    Color=<rgba 0.03 0.03 0.03 1>
    Lifetime=&Multiply_Float(
     A=&RandomRangeFloat(Minimum=0.8, Maximum=1)
     B=&Lerp_Float(A=1, B=0.6, Alpha=$System.InterpolateOverTime.InterpolateOverTime_Float.Moving Average)
    )
    Sprite Rotation Mode=::NewEnumerator2
    Sprite Size Mode=::NewEnumerator1
    Uniform Sprite Size Max=70
    Uniform Sprite Size Min=60
    UsePositionOffset=false
    Write Material Random=true
   )
   &self:SampleSourceSpawnGroupFromOtherParticle(PAR=DataInterface() {propertyValues {EmitterBinding {bindingMode=Other, emitterName=SKULLS}}}) {module=SampleSourceSpawnGroupFromOtherParticle}
  ]
  particleUpdateScript [
   &self:SamplePositionFromOtherParticle(PAR=DataInterface() {propertyValues {EmitterBinding {bindingMode=Other, emitterName=SKULLS}}}) {module=SamplePositionFromOtherParticle}
   /SetVariables(Particles.Position=&AddVectorToPosition(
     Position=$Particles.Position
     Vector=&MakeVector(
      X=0
      Y=0
      Z=&Lerp_Float(
       A=0
       B=&Lerp_Float(
        A=200
        B=&FN_FrontEndFloatModifier(Gameplay=180, Front End=100)
        Alpha=$System.InterpolateOverTime.InterpolateOverTime_Float.Moving Average
       )
       Alpha=$Particles.NormalizedAge
      )
     )
    )) {isSetParameters=true, module=SetVariables_#6}
   &ParticleState()
   &ScaleSpriteSize(
    Uniform Curve Scale=&Lerp_Float(A=1, B=1.2, Alpha=$System.InterpolateOverTime.InterpolateOverTime_Float.Moving Average)
    Uniform Curve Sprite Scale=DataInterface() {propertyValues {Curve {keys [
        {arriveTangent=1.6237, interpMode=RCIM_Cubic, leaveTangent=1.6237, tangentMode=RCTM_User, time=0, value=0.5}
        {arriveTangent=0.637, tangentMode=RCTM_User, time=1, value=1.4}
       ]}}}
   )
   &SpriteRotationRate(Rotation Rate=&RandomRangeFloat(Minimum=-150, Maximum=150))
   &SolveForcesAndVelocity(Clamp Velocity=false, Limit Acceleration=false)
   &ScaleColor(
    Scale Alpha=&FloatFromCurve(
     FloatCurve=DataInterface() {propertyValues {Curve {keys [
         {arriveTangent=12.0751, interpMode=RCIM_Cubic, leaveTangent=12.0751, tangentMode=RCTM_User, time=0, value=0}
         <at 0.2 1.3 RCIM_Cubic>
         <at 1 0 RCIM_Cubic>
        ]}}}
     Scale Curve=&Multiply_Float(
      A=&Lerp_Float(
       A=1
       B=&FN_FrontEndFloatModifier(Gameplay=1, Front End=0.6)
       Alpha=$System.InterpolateOverTime.InterpolateOverTime_Float.Moving Average
      )
      B=&FloatFromCurve(FloatCurve=DataInterface() {propertyValues {Curve {keys [<at 0 0 RCIM_Cubic>,<at 0.4 1 RCIM_Cubic>]}}}, CurveIndex=$System.Age)
     )
    )
    ScaleA=true
    ScaleRGB=true
   )
  ]
  rendererClasses [&NiagaraSpriteRendererProperties]
  renderers [{rendererClass=&NiagaraSpriteRendererProperties}]
 }
 Smoke_Green {
  emitterUpdateScript [
   &EmitterState(
    Delay First Loop Only=true
    Inactive Response=::NewEnumerator1
    Life Cycle Mode=::NewEnumerator1
    Loop Delay=$System.Delay
    Loop Duration=1
    UseLoopDelay=true
   )
   &SpawnPerFrame(Spawn:NiagaraBool=$System.DoOnce.Execute, Spawn Count=7, Use Spawn Probability=false)
  ]
  particleSpawnScript [
   &FN_SetSpawnGroup(Number of Spawn Groups=6)
   &InitializeParticle(
    Color=<rgba 0.0266 0.2 0.0671 0.222>
    Lifetime Max=0.4
    Lifetime Min=0.35
    Lifetime Mode=::NewEnumerator1
    Sprite Rotation Mode=::NewEnumerator2
    Sprite Size Mode=::NewEnumerator1
    Sprite UV Mode=::NewEnumerator3
    Uniform Sprite Size Max=60
    Uniform Sprite Size Min=60
    UsePositionOffset=false
    Write Material Random=true
   )
   &self:SampleSourceSpawnGroupFromOtherParticle(PAR=DataInterface() {propertyValues {EmitterBinding {bindingMode=Other, emitterName=SKULLS}}}) {module=SampleSourceSpawnGroupFromOtherParticle}
   &ShapeLocation(MaskSpawnGroup=false, Override Seed=false, Sphere Radius=2, UseOwnerScale=true)
  ]
  particleUpdateScript [
   &self:SamplePositionFromOtherParticle(PAR=DataInterface() {propertyValues {EmitterBinding {bindingMode=Other, emitterName=SKULLS}}}) {module=SamplePositionFromOtherParticle}
   /SetVariables(Particles.Position=&AddVectorToPosition(
     Position=$Particles.Position
     Vector=&MakeVector(
      X=0
      Y=0
      Z=&Lerp_Float(A=-5, B=&RandomRangeFloat(Minimum=40, Maximum=60), Alpha=$Particles.NormalizedAge)
     )
    )) {isSetParameters=true}
   &ParticleState()
   &SpriteRotationRate(Rotation Rate=&RandomRangeFloat(Minimum=-100, Maximum=100))
   &ScaleSpriteSize(Uniform Curve Sprite Scale=DataInterface() {propertyValues {Curve {keys [{arriveTangent=0.9211, interpMode=RCIM_Cubic, leaveTangent=0.9211, tangentMode=RCTM_User, time=0, value=0.4}, <at 1 1 RCIM_Cubic>]}}})
   &Drag(Drag=2, Use Linear Drag=true, Use Rotational Drag=true)
   &SolveForcesAndVelocity(Clamp Velocity=false, Limit Acceleration=false)
   &ScaleColor(
    Scale Alpha=&FloatFromCurve(
     FloatCurve=DataInterface() {propertyValues {Curve {keys [
         {interpMode=RCIM_Cubic, tangentMode=RCTM_User, time=0, value=0}
         {arriveTangent=-3.5829, interpMode=RCIM_Cubic, leaveTangent=-3.5829, tangentMode=RCTM_User, time=0.15, value=1}
         <at 1 0 RCIM_Cubic>
        ]}}}
     Scale Curve=&Multiply_Float(
      A=&Lerp_Float(A=1, B=1, Alpha=$System.InterpolateOverTime.InterpolateOverTime_Float.Moving Average)
      B=&MakeCustomFloatFromBool(
       Bool=&SetBoolByIntComparison(A=$Particles.PartitionParticles.Partition, B=0, Comparison Type=::NewEnumerator0)
       True Float=&Lerp_Float(A=1, B=0, Alpha=$System.InterpolateOverTime004.InterpolateOverTime_Float.Moving Average)
       False Float=&Lerp_Float(A=1, B=0, Alpha=$System.InterpolateOverTime005.InterpolateOverTime_Float.Moving Average)
      )
     )
    )
    Scale RGB=&MakeVectorFromLinearColorRGB(Color=&ColorFromCurve(ColorCurve=DataInterface() {propertyValues {
        AlphaCurve {keys []}
        BlueCurve {keys [<at 0 1 RCIM_Cubic>,<at 0.3549 0.2 RCIM_Cubic>]}
        GreenCurve {keys [<at 0 1 RCIM_Cubic>,<at 0.3549 0.2 RCIM_Cubic>]}
        RedCurve {keys [<at 0 1 RCIM_Cubic>,<at 0.3549 0.2 RCIM_Cubic>]}
       }}))
    ScaleA=true
    ScaleRGB=false
   )
   &SubUVAnimation(
    End Frame Range Override=30
    "Sprite Renderer "=DataInterface() {propertyValues {SpriteRenderer=&self:Smoke_Green_11.NiagaraSpriteRendererProperties_0}}
    Start Frame Range Override=15
    UseEndFrame=true
    UseStartFrame=true
   )
  ]
  rendererClasses [&NiagaraSpriteRendererProperties]
  renderers [{rendererClass=&NiagaraSpriteRendererProperties}]
 }
 Sparks {
  emitterUpdateScript [
   &EmitterState()
   &SpawnParticlesFromOtherEmitter(
    Attribute Reader=DataInterface() {propertyValues {EmitterBinding {bindingMode=Other, emitterName=SKULLS}}}
    Spawn Rate=&Multiply_Float(
     A=&Lerp_Float(A=6, B=10, Alpha=$System.InterpolateOverTime.InterpolateOverTime_Float.Moving Average)
     B=&Multiply_Float(
      A=&MakeCustomFloatFromBool(Bool=$System.DoOnce.Execute, True Float=30, False Float=1)
      B=&Lerp_Float(
       A=1
       B=10
       Alpha=&FloatFromCurve(FloatCurve=DataInterface() {propertyValues {Curve {keys [<at 0 1>,<at 1 0>]}}}, CurveIndex=$Engine.System.Age)
      )
     )
    )
    Spawn Rate Per Particle Cap=100000
    Use Maximum Age=false
    Use Minimum Age=false
   )
  ]
  particleSpawnScript [
   &PartitionParticles()
   &InitializeParticle(
    Color=<rgba 0.5647 5 1.5996 1>
    Lifetime Max=0.5
    Lifetime Min=0.5
    Lifetime Mode=::NewEnumerator1
    Sprite Size Max=<1 7>
    Sprite Size Min=<0.7 3>
    Sprite Size Mode=::NewEnumerator2
    UsePositionOffset=false
    Write Material Random=true
   )
   &SampleParticlesFromOtherEmitter(Velocity Application:ENiagara_DirectReadApplicationMode=::NewEnumerator0, Velocity Sampling=::NewEnumerator0)
   &ShapeLocation(
    MaskSpawnGroup=false
    Override Seed=false
    Sphere Radius=&Lerp_Float(A=12, B=9, Alpha=$System.InterpolateOverTime.InterpolateOverTime_Float.Moving Average)
    UseOwnerScale=true
   )
   &AddVelocity(
    Cone Angle=&Lerp_Float(A=25, B=15, Alpha=$System.InterpolateOverTime.InterpolateOverTime_Float.Moving Average)
    Cone Axis=&SelectVectorFromArray(
     Direct Array Index=$Particles.PartitionParticles.Partition
     Vector Selection Array=DataInterface() {propertyValues {FloatData [<0 0 0.7>, <0 0 0.7>]}}
    )
    Module.UseConeFalloff=false
    Module.UseVelDistribution=false
    Override Seed=false
    Velocity Mode=::NewEnumerator2
    Velocity Speed=&Multiply_Float(
     A=&RandomRangeFloat(Minimum=150, Maximum=600)
     B=&Lerp_Float(A=1, B=2, Alpha=$System.InterpolateOverTime.InterpolateOverTime_Float.Moving Average)
    )
   )
  ]
  particleUpdateScript [
   &ParticleState()
   /SetVariables(Particles.Glow03_Color=&MakeLinearColorFromVectorAndFloat(
     "Vector (RGB)"=&MakeVectorFromLinearColorRGB(Color=&MultiplyLinearColor(
       Linear Color=<rgba 0.0335 1 0.2591 1>
       Scale Factor=&MakeVector4_FromFloat(Float=&Lerp_Float(A=0.3, B=1, Alpha=$System.InterpolateOverTime.InterpolateOverTime_Float.Moving Average))
      ))
     "Float (Alpha)"=&Multiply_Float(A=0.2, B=1)
    )) {isSetParameters=true}
   /SetVariables(Particles.Glow03_Size=&Multiply_Vector2D_ByFloat(
     Float=&FloatFromCurve(
      FloatCurve=DataInterface() {propertyValues {Curve {keys [
          <at 0 0 RCIM_Cubic>
          <at 0.2 1.3 RCIM_Cubic>
          {arriveTangent=-0.1423, interpMode=RCIM_Cubic, leaveTangent=-0.1423, tangentMode=RCTM_User, time=1, value=0}
         ]}}}
      Scale Curve=8
     )
     Vector2D=<1 4>
    )) {isSetParameters=true}
   &ScaleColor(
    Scale Alpha=&FloatFromCurve(
     FloatCurve=DataInterface() {propertyValues {Curve {keys [<at 0 0>,<at 0.3 1>,<at 1 1>]}}}
     Scale Curve=&MakeCustomFloatFromBool(
      Bool=&SetBoolByIntComparison(A=$Particles.PartitionParticles.Partition, B=0, Comparison Type=::NewEnumerator0)
      True Float=&Lerp_Float(A=1, B=0, Alpha=$System.InterpolateOverTime005.InterpolateOverTime_Float.Moving Average)
      False Float=&Lerp_Float(A=1, B=0, Alpha=$System.InterpolateOverTime004.InterpolateOverTime_Float.Moving Average)
     )
    )
    Scale RGB=&Multiply_VectorByFloat(
     Vector=<1 1 1>
     Float=&Lerp_Float(A=0.1, B=1, Alpha=$System.InterpolateOverTime.InterpolateOverTime_Float.Moving Average)
    )
    ScaleA=true
    ScaleRGB=true
   )
   &ScaleSpriteSize(Scale Factor=&Vector2DFromFloat(Value=&FloatFromCurve(
      FloatCurve=DataInterface() {propertyValues {Curve {keys [
          <at 0 0 RCIM_Cubic>
          {interpMode=RCIM_Cubic, leaveTangent=-3.0296, tangentMode=RCTM_Break, time=0.2, value=1.2}
          {arriveTangent=-1.8107, interpMode=RCIM_Cubic, leaveTangent=-1.8107, tangentMode=RCTM_User, time=1, value=0}
         ]}}}
      Scale Curve=1.8
     ))) {module=ScaleSpriteSize001}
   &CurlNoiseForce(Noise Frequency=500, Noise Strength=500, Pan Noise=false)
   &Drag(Use Linear Drag=true, Use Rotational Drag=true)
   &SolveForcesAndVelocity(Clamp Velocity=false, Limit Acceleration=false)
   &ScaleSpriteSizeBySpeed(Max Scale Factor=<1 2>)
  ]
  rendererClasses [&NiagaraSpriteRendererProperties]
  renderers [
   {rendererClass=&NiagaraSpriteRendererProperties}
   {rendererClass=&NiagaraSpriteRendererProperties, rendererIndex=1}
  ]
 }
 Trails {
  emitterUpdateScript [
   &EmitterState(
    Inactive Response=::NewEnumerator1
    Life Cycle Mode=::NewEnumerator1
    Loop Behavior=::NewEnumerator1
    Loop Duration Mode=::NewEnumerator1
    UseLoopDelay=false
   )
   &SpawnBurst_Instantaneous(Spawn Count=6, Use Spawn Probability=false, UseLoopCountLimit=false)
  ]
  particleSpawnScript [
   &FN_SetSpawnGroup(Number of Spawn Groups=6)
   /SetVariables(Particles.BoneIndex:NiagaraInt32=/Niagara/DynamicInputs/Divide/Divide_Int(A:NiagaraInt32=&ReturnExecIndex(), B:NiagaraInt32=2)) {isSetParameters=true}
   &InitializeParticle(
    Color=<rgba 0.0097 0.0097 0.0097 1>
    Lifetime=3
    Mesh Scale=&RandomRangeVector(Minimum=<1 1 2.6>, Maximum=<1 1 3>)
    Mesh Scale Mode=::NewEnumerator4
    UsePositionOffset=false
    Write Material Random=true
   )
   &InitialMeshOrientation(Mesh Axis To Orient=<0 0 1>, Mesh Orientation Mode=::NewEnumerator2, Orientation Vector=<0 0 1>, Use Rotation Vector=false)
   &self:SampleSourceSpawnGroupFromOtherParticle(PAR=DataInterface() {propertyValues {EmitterBinding {bindingMode=Other, emitterName=SKULLS}}}) {module=SampleSourceSpawnGroupFromOtherParticle}
  ]
  particleUpdateScript [
   /SetVariables(Particles.EnergyTrailColor:LinearColor=&MultiplyLinearColor(
     Linear Color=<rgba 0 4 0.974 1.2>
     Scale Factor=&MakeFloat4FromVectorAndFloat(
      "Vector (RGB)"=&VectorFromFloat(Value=&Multiply_Float(
        A=&Multiply_Float(
         A=&Lerp_Float(
          A=&FN_FrontEndFloatModifier(Gameplay=1, Front End=1.5)
          B=&FN_FrontEndFloatModifier(Gameplay=6, Front End=5)
          Alpha=$System.InterpolateOverTime.InterpolateOverTime_Float.Moving Average
         )
         B=&FloatFromCurve(
          FloatCurve=DataInterface() {propertyValues {Curve {keys [
              <at 0 1 RCIM_Cubic>
              {arriveTangent=5.8588, interpMode=RCIM_Cubic, leaveTangent=5.8588, time=0.2, value=7}
              {arriveTangent=-12.5901, interpMode=RCIM_Cubic, leaveTangent=-12.5901, tangentMode=RCTM_User, time=0.3453, value=3.0232}
              <at 0.8 1 RCIM_Cubic>
             ]}}}
          CurveIndex=$System.Timeline003.NormalizedTime
         )
        )
        B=&FN_FrontEndFloatModifier(
         Gameplay=&FloatFromCurve(FloatCurve=DataInterface() {propertyValues {Curve {keys [<at 0.6 8 RCIM_Cubic>,<at 1.5 1 RCIM_Cubic>]}}}, CurveIndex=$Engine.System.Age)
         Front End=&FloatFromCurve(
          FloatCurve=DataInterface() {propertyValues {Curve {keys [<at 0.15 1 RCIM_Cubic>,<at 0.8 3.5 RCIM_Cubic>,<at 1.5 1 RCIM_Cubic>]}}}
          CurveIndex=$Engine.System.Age
         )
        )
       ))
      "Float (Alpha)"=&FloatFromCurve(FloatCurve=DataInterface() {propertyValues {Curve {keys [<at 0 0 RCIM_Cubic>,<at 0.4 1 RCIM_Cubic>]}}}, CurveIndex=$System.Age)
     )
    )) {isSetParameters=true}
   &ParticleState(Kill Particles When Lifetime Has Elapsed=false, Let Infinitely Lived Particles Die When Emitter Deactivates=true)
   &ScaleMeshSize(Scale Factor=&MakeVector(X=1, Y=1, Z=&Lerp_Float(A=0.9, B=1.1, Alpha=$System.InterpolateOverTime.InterpolateOverTime_Float.Moving Average))) {module=ScaleMeshSize001}
   &self:SamplePositionFromOtherParticle(PAR=DataInterface() {propertyValues {EmitterBinding {bindingMode=Other, emitterName=SKULLS}}}) {module=SamplePositionFromOtherParticle}
   /SetVariables(Particles.SkullScale:NiagaraFloat=&Multiply_Float(A=$Particles.SkullScale, B=0.8)) {isSetParameters=true}
   /SetVariables(Particles.Position=&AddVectorToPosition(
     Position=$Particles.Position
     Vector=&TransformVector(
      Vector=&Multiply_VectorByFloat(
       Vector=&Lerp_Vector(A=<0 0 0>, B=<0 1 -5>, Alpha=$System.InterpolateOverTime.InterpolateOverTime_Float.Moving Average)
       Float=$Particles.SkullScale
      )
      Destination Space=::Simulation
     )
    )) {isSetParameters=true}
   &InterpolateOverTime(Rate Of Change=5, Target Vector Value:NiagaraPosition=$Particles.Position, Variable Type=::NewEnumerator2)
   &DynamicMaterialParameters(
    Index 0 Param 1=&Multiply_Float(
     A=&FN_FrontEndFloatModifier(
      Gameplay=&Lerp_Float(A=30, B=50, Alpha=$System.InterpolateOverTime.InterpolateOverTime_Float.Moving Average)
      Front End=&Lerp_Float(A=40, B=38, Alpha=$System.InterpolateOverTime.InterpolateOverTime_Float.Moving Average)
     )
     B=$Particles.SkullScale
    )
    Index 0 Param 2=&Multiply_Float(
     A=&FN_FrontEndFloatModifier(
      Gameplay=&Lerp_Float(A=120, B=90, Alpha=$System.InterpolateOverTime.InterpolateOverTime_Float.Moving Average)
      Front End=&Lerp_Float(A=140, B=80, Alpha=$System.InterpolateOverTime.InterpolateOverTime_Float.Moving Average)
     )
     B=$Particles.SkullScale
    )
    Index 0 Param 3=1
    Index 0 Param 4=&Add_Float(
     A=&RandomRangeFloat(Minimum=0, Maximum=1)
     B=&Multiply_Float(A=$System.FN_Integrator002.Value, B=-1)
    )
    Index 1 Param 1=&Lerp_Float(A=0, B=1, Alpha=$System.InterpolateOverTime.InterpolateOverTime_Float.Moving Average)
    Index 1 Param 2=&FloatFromCurve(FloatCurve=DataInterface() {propertyValues {Curve {keys [{tangentMode=RCTM_User, time=0, value=0}, {tangentMode=RCTM_User, time=0.35, value=1}]}}})
    Index 1 Param 3=&FN_FrontEndFloatModifier(
     Gameplay=0
     Front End=&Lerp_Float(A=-70, B=-10, Alpha=$System.InterpolateOverTime.InterpolateOverTime_Float.Moving Average)
    )
    Index 1 Param 4=&FN_FrontEndFloatModifier(
     Gameplay=&Lerp_Float(A=1, B=0, Alpha=$System.InterpolateOverTime.InterpolateOverTime_Float.Moving Average)
     Front End=1
    )
    Index 2 Param 1=1
    Index 2 Param 2=&Lerp_Float(
     A=100
     B=250
     Alpha=&ClampFloat(Float=&Add_Float(
       A=&RemapRange(Input Value To Remap=$Particles.RotAmount, Input Min=15, Input Max=90, Output Min=0, Output Max=1)
       B=$System.Timeline004.NormalizedTime
      ))
    )
    Index 3 Param 1=&MakeFloatFromVector(Vector=$StackContext.InterpolateOverTime.InterpolateOverTime_V3_.Moving Average, Channel=::NewEnumerator0)
    Index 3 Param 2=&MakeFloatFromVector(Vector=$StackContext.InterpolateOverTime.InterpolateOverTime_V3_.Moving Average, Channel=::NewEnumerator1)
    Index 3 Param 3=&MakeFloatFromVector(Vector=$StackContext.InterpolateOverTime.InterpolateOverTime_V3_.Moving Average, Channel=::NewEnumerator2)
    Index 3 Param 4=1
    Module.Index0Param0Write=true
    Module.Index0Param1Write=true
    Module.Index0Param2Write=true
    Module.Index0Param3Write=true
    Module.Index1Param0Write=true
    Module.Index1Param1Write=true
    Module.Index1Param2Write=true
    Module.Index1Param3Write=true
    Module.Index2Param0Write=true
    Module.Index2Param1Write=true
    Module.Index2Param2Write=false
    Module.Index2Param3Write=false
    Module.Index3Param0Write=true
    Module.Index3Param1Write=true
    Module.Index3Param2Write=true
    Module.Index3Param3Write=true
    Write Parameter Index 1=true
    Write Parameter Index 2=true
    Write Parameter Index 3=true
   )
   &ScaleColor(
    Scale Alpha=&FloatFromCurve(FloatCurve=DataInterface() {propertyValues {Curve {keys [<at 0 0 RCIM_Cubic>,<at 0.4 1 RCIM_Cubic>]}}}, CurveIndex=$System.Age)
    ScaleA=true
    ScaleRGB=false
   )
  ]
  rendererClasses [&NiagaraMeshRendererProperties]
  renderers [
   {rendererClass=&NiagaraMeshRendererProperties, rendererIndex=0}
   {rendererClass=&NiagaraMeshRendererProperties, rendererIndex=1}
  ]
 }
}
summary {
 systemName=NS_DarkStance_Contrail
 emitters [
  {emitterName=Sparks, rendererClasses [&NiagaraSpriteRendererProperties]}
  {emitterName=Smoke_Green, rendererClasses [&NiagaraSpriteRendererProperties]}
  {emitterName=Smoke, rendererClasses [&NiagaraSpriteRendererProperties]}
  {emitterName=Trails, rendererClasses [&NiagaraMeshRendererProperties]}
  {emitterName=Blend_Smoke, rendererClasses [&NiagaraSpriteRendererProperties]}
  {emitterName=SKULLS, rendererClasses [&NiagaraMeshRendererProperties]}
 ]
 userVariables [
  User.bIsFrontEnd {type=NiagaraBool}
  User.bIsFrontEndPreview {type=NiagaraBool}
  User.ParaGlideLeanAlpha {type=NiagaraFloat, defaultValue=0, description=""}
  User.RibbonIDOffset {type=NiagaraInt32}
 ]
}
